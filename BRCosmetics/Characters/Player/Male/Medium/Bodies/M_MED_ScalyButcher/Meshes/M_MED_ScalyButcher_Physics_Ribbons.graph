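# Physics asset for the front and back ribbons on the right side of the M_MED_ScalyButcher male player cosmetic body.
# Capsules cover the pelvis, right thigh and calf, plus two segments per ribbon.
# Each ribbon chains from the pelvis through zero-size skirt base and 01 bodies, which act as links without volume.
# Joints down to each first segment swing and twist freely, while second segments have limited swing, tighter at the front, and zero twist.
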PhysicsAsset /BRCosmetics/Characters/Player/Male/Medium/Bodies/M_MED_ScalyButcher/Meshes/M_MED_ScalyButcher_Physics_Ribbons
bodies {
 calf_r [{shapeType=Capsule, length=36.7843, radius=12.5226, shapeName=calf_r_capsule, center=<22.7205 0 0>, rotation=<rot 90 0 0>}]
 dyn_ribbon_bk_01 [{shapeType=Capsule, length=6.6608, radius=5.414, shapeName=dyn_ribbon_bk_01_capsule, center=<-8.1463 0 0>, rotation=<rot 90 0 0>}]
 dyn_ribbon_bk_02 [{
   shapeType=Capsule
   length=11.6379
   radius=7.1587
   shapeName=dyn_ribbon_bk_02_capsule
   center=<-7.8592 -0.0753 -1.4563>
   rotation=<rot 5.082 -179.1804 -80.3302>
  }]
 dyn_ribbon_frnt_01 [{shapeType=Capsule, length=4.6613, radius=4.1478, shapeName=dyn_ribbon_frnt_01_capsule, center=<-6.3391 0 0>, rotation=<rot 90 0 0>}]
 dyn_ribbon_frnt_02 [{
   shapeType=Capsule
   length=24.1268
   radius=6.1672
   shapeName=dyn_ribbon_frnt_02_capsule
   center=<-7.5296 -1.9796 -2.903>
   rotation=<rot -3.3727 -4.8394 -85.0546>
  }]
 dyn_skirt_bk_r_01 [{shapeType=Capsule, length=0, radius=0, shapeName=dyn_skirt_bk_r_01_capsule}]
 dyn_skirt_bk_r_base [{shapeType=Capsule, length=0, radius=0, shapeName=dyn_skirt_bk_r_base_capsule}]
 dyn_skirt_frnt_r_01 [{shapeType=Capsule, length=0, radius=0, shapeName=dyn_skirt_frnt_r_01_capsule}]
 dyn_skirt_frnt_r_base [{shapeType=Capsule, length=0, radius=0, shapeName=dyn_skirt_frnt_r_base_capsule}]
 pelvis [{shapeType=Capsule, length=5.3874, radius=15.6554, center=<-6.7862 0 0>}]
 thigh_r [{shapeType=Capsule, length=42.7956, radius=11.5277, center=<22.6754 0 -0.0545>, rotation=<rot 90 0 0>}]
}
constraints [
 {bone1Name=thigh_r, bone2Name=pelvis, swing1LimitDegrees=45, swing2LimitDegrees=45, twistLimitDegrees=45}
 {
  bone1Name=calf_r
  bone2Name=thigh_r
  swing1LimitDegrees=45
  swing1Motion=Free
  swing2LimitDegrees=45
  swing2Motion=Free
  twistLimitDegrees=45
  twistMotion=Free
 }
 {
  bone1Name=dyn_ribbon_bk_01
  bone2Name=dyn_skirt_bk_r_01
  swing1LimitDegrees=45
  swing1Motion=Free
  swing2LimitDegrees=45
  swing2Motion=Free
  twistLimitDegrees=45
  twistMotion=Free
 }
 {bone1Name=dyn_ribbon_bk_02, bone2Name=dyn_ribbon_bk_01, swing1LimitDegrees=45, swing2LimitDegrees=45, twistLimitDegrees=0}
 {
  bone1Name=dyn_ribbon_frnt_01
  bone2Name=dyn_skirt_frnt_r_01
  swing1LimitDegrees=45
  swing1Motion=Free
  swing2LimitDegrees=45
  swing2Motion=Free
  twistLimitDegrees=45
  twistMotion=Free
 }
 {bone1Name=dyn_ribbon_frnt_02, bone2Name=dyn_ribbon_frnt_01, swing1LimitDegrees=10, swing2LimitDegrees=40, twistLimitDegrees=0}
 {
  bone1Name=dyn_skirt_bk_r_base
  bone2Name=pelvis
  swing1LimitDegrees=45
  swing1Motion=Free
  swing2LimitDegrees=45
  swing2Motion=Free
  twistLimitDegrees=45
  twistMotion=Free
 }
 {
  bone1Name=dyn_skirt_bk_r_01
  bone2Name=dyn_skirt_bk_r_base
  swing1LimitDegrees=45
  swing1Motion=Free
  swing2LimitDegrees=45
  swing2Motion=Free
  twistLimitDegrees=45
  twistMotion=Free
 }
 {
  bone1Name=dyn_skirt_frnt_r_base
  bone2Name=pelvis
  swing1LimitDegrees=45
  swing1Motion=Free
  swing2LimitDegrees=45
  swing2Motion=Free
  twistLimitDegrees=45
  twistMotion=Free
 }
 {
  bone1Name=dyn_skirt_frnt_r_01
  bone2Name=dyn_skirt_frnt_r_base
  swing1LimitDegrees=45
  swing1Motion=Free
  swing2LimitDegrees=45
  swing2Motion=Free
  twistLimitDegrees=45
  twistMotion=Free
 }
]
