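# Physics asset for the skirt of the male medium ScalyButcher player body, modelling it as jointed capsule chains around the lower body.
# Five chains (front right, side left, three back) each run from a pelvis-linked base through four segments with tight swing and twist limits.
# Angular-free cross constraints tie neighbouring chains at the lower three levels, linking the panels sideways.
# Pelvis and leg capsules, plus a large sphere below root, supply body shapes.
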PhysicsAsset /BRCosmetics/Characters/Player/Male/Medium/Bodies/M_MED_ScalyButcher/Meshes/M_MED_ScalyButcher_Physics_Skirt
bodies {
 calf_l [{shapeType=Capsule, length=32.6849, radius=11.6844, center=<-20.2002 0 0.9072>, rotation=<rot 90 0 0>}]
 calf_r [{shapeType=Capsule, length=25.1868, radius=11.2977, center=<20.2002 0.1412 -1.3484>, rotation=<rot 90 0 180>}]
 dyn_skirt_bk_l_01 [{
   shapeType=Capsule
   length=11.9581
   radius=4.1153
   shapeName=dyn_skirt_bk_l_01_capsule
   center=<6.8042 0.0685 3.7194>
   rotation=<rot -0.0002 6.5827 89.9999>
  }]
 dyn_skirt_bk_l_02 [{
   shapeType=Capsule
   length=11.6125
   radius=5.4609
   shapeName=dyn_skirt_bk_l_02_capsule
   center=<6.2107 0 3.314>
   rotation=<rot -0.0002 6.5827 89.9999>
  }]
 dyn_skirt_bk_l_03 [{
   shapeType=Capsule
   length=11.0734
   radius=7
   shapeName=dyn_skirt_bk_l_03_capsule
   center=<6.2107 0 4.0206>
   rotation=<rot -0.0002 6.5827 89.9999>
  }]
 dyn_skirt_bk_l_04 [{
   shapeType=Capsule
   length=24.2966
   radius=8.8649
   shapeName=dyn_skirt_bk_l_04_capsule
   center=<8.6224 0 6.7036>
   rotation=<rot -0.0002 6.5827 89.9999>
  }]
 dyn_skirt_bk_l_base [{shapeType=Capsule, length=0, radius=0, shapeName=dyn_skirt_bk_l_base_capsule}]
 dyn_skirt_bk_r_01 [{
   shapeType=Capsule
   length=12.5754
   radius=4.7411
   shapeName=dyn_skirt_bk_r_01_capsule
   center=<-10.7155 0 -6.5663>
   rotation=<rot 0 0.0001 80.0362>
  }]
 dyn_skirt_bk_r_02 [{
   shapeType=Capsule
   length=11.64
   radius=6.6765
   shapeName=dyn_skirt_bk_r_02_capsule
   center=<-7.3273 0 -6.9177>
   rotation=<rot 0 0.0001 80.0362>
  }]
 dyn_skirt_bk_r_03 [{
   shapeType=Capsule
   length=12.5211
   radius=8.6446
   shapeName=dyn_skirt_bk_r_03_capsule
   center=<-7.9311 0 -5.8218>
   rotation=<rot -0.0001 -6.8683 80.0362>
  }]
 dyn_skirt_bk_r_04 [{
   shapeType=Capsule
   length=21.9422
   radius=9.6148
   shapeName=dyn_skirt_bk_r_04_capsule
   center=<-9.7222 0.4154 -7.4353>
   rotation=<rot -0.0001 -6.8683 80.0362>
  }]
 dyn_skirt_bk_r_base [{shapeType=Capsule, length=0, radius=0, shapeName=dyn_skirt_bk_r_base_capsule}]
 dyn_skirt_frnt_r_01 [{
   shapeType=Capsule
   length=9.049
   radius=9.1078
   shapeName=dyn_skirt_frnt_r_01_capsule
   center=<-7.6052 -0.9601 -8.4953>
   rotation=<rot 0 0 90>
  }]
 dyn_skirt_frnt_r_03 [{shapeType=Capsule, length=14.0398, radius=8.6273, shapeName=dyn_skirt_frnt_r_03_capsule, center=<-7.9739 0 -7>, rotation=<rot 0 0 90>}]
 dyn_skirt_frnt_r_04 [{shapeType=Capsule, length=14.115, radius=8.8989, shapeName=dyn_skirt_frnt_r_04_capsule, center=<-10.4675 0 -7.0389>, rotation=<rot 0 0 90>}]
 dyn_skirt_frnt_r_2 [{shapeType=Capsule, length=16.2496, radius=6.3598, shapeName=dyn_skirt_frnt_r_2_capsule, center=<-7.6249 0 -6.425>, rotation=<rot 0 0 90>}]
 dyn_skirt_frnt_r_base [{shapeType=Capsule, length=0, radius=0, shapeName=dyn_skirt_frnt_r_base_capsule}]
 dyn_skirt_mid_bk_01 [{
   shapeType=Capsule
   length=17.1107
   radius=4.1153
   shapeName=dyn_skirt_mid_bk_01_capsule
   center=<8.863 -3.1093 0>
   rotation=<rot 2.4018 0.0003 0>
  }]
 dyn_skirt_mid_bk_02 [{
   shapeType=Capsule
   length=16.765
   radius=5.4609
   shapeName=dyn_skirt_mid_bk_02_capsule
   center=<6.9771 -4.2263 -0.0791>
   rotation=<rot 2.4018 0.0003 0>
  }]
 dyn_skirt_mid_bk_03 [{
   shapeType=Capsule
   length=21.885
   radius=7
   shapeName=dyn_skirt_mid_bk_03_capsule
   center=<7.0386 -4.5872 -0.0765>
   rotation=<rot 2.4018 0.0003 0>
  }]
 dyn_skirt_mid_bk_04 [{
   shapeType=Capsule
   length=25.6924
   radius=9.0507
   shapeName=dyn_skirt_mid_bk_04_capsule
   center=<8.9671 -6.3726 0>
   rotation=<rot 13.7745 0.0005 0>
  }]
 dyn_skirt_mid_bk_base [{shapeType=Capsule, length=0, radius=0, shapeName=dyn_skirt_mid_bk_base_capsule}]
 dyn_skirt_side_l_01 [{
   shapeType=Capsule
   length=7.4544
   radius=9.051
   shapeName=dyn_skirt_side_l_01_capsule
   center=<4.3905 -0.1248 8.6015>
   rotation=<rot 0 5.3189 90>
  }]
 dyn_skirt_side_l_02 [
  {
   shapeType=Capsule
   length=9.1692
   radius=8.3362
   shapeName=dyn_skirt_side_l_02_capsule
   center=<4.848 -0.0823 7.6677>
   rotation=<rot 0 5.3189 90>
  }
  {shapeType=Capsule, length=12, radius=3, shapeName=dyn_skirt_side_l_02_capsule, center=<8.0963 9.5185 -0.079>, rotation=<rot 90 0 0>}
 ]
 dyn_skirt_side_l_03 [{
   shapeType=Capsule
   length=10.9125
   radius=7.5929
   shapeName=dyn_skirt_side_l_03_capsule
   center=<6.5716 0.0782 6.2256>
   rotation=<rot 0 5.3189 90>
  }]
 dyn_skirt_side_l_04 [{
   shapeType=Capsule
   length=12.5057
   radius=6.9998
   shapeName=dyn_skirt_side_l_04_capsule
   center=<9.8558 0 3.0057>
   rotation=<rot 0 5.3189 90>
  }]
 dyn_skirt_side_l_base [{shapeType=Capsule, length=0, radius=0, shapeName=dyn_skirt_side_l_base_capsule}]
 foot_l [{shapeType=Capsule, length=12.7957, radius=8.1395, center=<-4.6409 5.1059 0>, rotation=<rot 90 180 87.7526>}]
 foot_r [{shapeType=Capsule, length=12.7957, radius=8.1395, center=<4.6409 -5.1059 0>, rotation=<rot 90 180 -92.2474>}]
 pelvis [{shapeType=Capsule, length=9.6354, radius=11.7125, center=<-4.0806 -0.0225 0>}]
 root [{shapeType=Sphere, radius=146.6037, shapeName=root_sphere, center=<0.1114 -6.0187 -137.6639>}]
 thigh_l [{shapeType=Capsule, length=36.1983, radius=10.135, center=<-20.2067 0 0.2164>, rotation=<rot 88.29 179.9997 179.9997>}]
 thigh_r [{shapeType=Capsule, length=33.0242, radius=10.135, center=<20.2748 0 -0.7841>, rotation=<rot 84.8458 179.9986 -0.0008>}]
}
constraints [
 {bone1Name=thigh_l, bone2Name=pelvis, swing1LimitDegrees=45, swing2LimitDegrees=45, twistLimitDegrees=45}
 {bone1Name=calf_l, bone2Name=thigh_l, swing1LimitDegrees=45, swing2LimitDegrees=45, twistLimitDegrees=45}
 {bone1Name=foot_l, bone2Name=calf_l, swing1LimitDegrees=45, swing2LimitDegrees=45, twistLimitDegrees=45}
 {bone1Name=foot_r, bone2Name=calf_r, swing1LimitDegrees=45, swing2LimitDegrees=45, twistLimitDegrees=45}
 {
  bone1Name=dyn_skirt_bk_l_base
  bone2Name=pelvis
  swing1LimitDegrees=45
  swing1Motion=Free
  swing2LimitDegrees=45
  swing2Motion=Free
  twistLimitDegrees=45
  twistMotion=Free
 }
 {bone1Name=dyn_skirt_bk_l_01, bone2Name=dyn_skirt_bk_l_base, swing1LimitDegrees=20, swing2LimitDegrees=20, twistLimitDegrees=5}
 {bone1Name=dyn_skirt_bk_l_02, bone2Name=dyn_skirt_bk_l_01, swing1LimitDegrees=20, swing2LimitDegrees=20, twistLimitDegrees=5}
 {bone1Name=dyn_skirt_bk_l_03, bone2Name=dyn_skirt_bk_l_02, swing1LimitDegrees=20, swing2LimitDegrees=20, twistLimitDegrees=5}
 {bone1Name=dyn_skirt_bk_l_04, bone2Name=dyn_skirt_bk_l_03, swing1LimitDegrees=20, swing2LimitDegrees=20, twistLimitDegrees=5}
 {
  bone1Name=dyn_skirt_bk_r_base
  bone2Name=pelvis
  swing1LimitDegrees=45
  swing1Motion=Free
  swing2LimitDegrees=45
  swing2Motion=Free
  twistLimitDegrees=45
  twistMotion=Free
 }
 {bone1Name=dyn_skirt_bk_r_01, bone2Name=dyn_skirt_bk_r_base, swing1LimitDegrees=20, swing2LimitDegrees=20, twistLimitDegrees=5}
 {bone1Name=dyn_skirt_bk_r_02, bone2Name=dyn_skirt_bk_r_01, swing1LimitDegrees=20, swing2LimitDegrees=20, twistLimitDegrees=5}
 {bone1Name=dyn_skirt_bk_r_03, bone2Name=dyn_skirt_bk_r_02, swing1LimitDegrees=20, swing2LimitDegrees=20, twistLimitDegrees=5}
 {bone1Name=dyn_skirt_bk_r_04, bone2Name=dyn_skirt_bk_r_03, swing1LimitDegrees=20, swing2LimitDegrees=20, twistLimitDegrees=10}
 {
  bone1Name=dyn_skirt_frnt_r_base
  bone2Name=pelvis
  swing1LimitDegrees=45
  swing1Motion=Free
  swing2LimitDegrees=45
  swing2Motion=Free
  twistLimitDegrees=45
  twistMotion=Free
 }
 {bone1Name=dyn_skirt_frnt_r_01, bone2Name=dyn_skirt_frnt_r_base, swing1LimitDegrees=20, swing2LimitDegrees=20, twistLimitDegrees=1}
 {bone1Name=dyn_skirt_frnt_r_2, bone2Name=dyn_skirt_frnt_r_01, swing1LimitDegrees=20, swing2LimitDegrees=20, twistLimitDegrees=1}
 {bone1Name=dyn_skirt_frnt_r_03, bone2Name=dyn_skirt_frnt_r_2, swing1LimitDegrees=20, swing2LimitDegrees=20, twistLimitDegrees=10}
 {bone1Name=dyn_skirt_frnt_r_04, bone2Name=dyn_skirt_frnt_r_03, swing1LimitDegrees=20, swing2LimitDegrees=30, twistLimitDegrees=10}
 {
  bone1Name=dyn_skirt_mid_bk_base
  bone2Name=pelvis
  swing1LimitDegrees=45
  swing1Motion=Free
  swing2LimitDegrees=45
  swing2Motion=Free
  twistLimitDegrees=45
  twistMotion=Free
 }
 {bone1Name=dyn_skirt_mid_bk_01, bone2Name=dyn_skirt_mid_bk_base, swing1LimitDegrees=20, swing2LimitDegrees=20, twistLimitDegrees=5}
 {bone1Name=dyn_skirt_mid_bk_02, bone2Name=dyn_skirt_mid_bk_01, swing1LimitDegrees=20, swing2LimitDegrees=20, twistLimitDegrees=5}
 {bone1Name=dyn_skirt_mid_bk_03, bone2Name=dyn_skirt_mid_bk_02, swing1LimitDegrees=20, swing2LimitDegrees=20, twistLimitDegrees=5}
 {bone1Name=dyn_skirt_mid_bk_04, bone2Name=dyn_skirt_mid_bk_03, swing1LimitDegrees=20, swing2LimitDegrees=20, twistLimitDegrees=5}
 {
  bone1Name=dyn_skirt_side_l_base
  bone2Name=pelvis
  swing1LimitDegrees=45
  swing1Motion=Free
  swing2LimitDegrees=45
  swing2Motion=Free
  twistLimitDegrees=45
  twistMotion=Free
 }
 {bone1Name=dyn_skirt_side_l_01, bone2Name=dyn_skirt_side_l_base, swing1LimitDegrees=10, swing2LimitDegrees=20, twistLimitDegrees=0}
 {bone1Name=dyn_skirt_side_l_02, bone2Name=dyn_skirt_side_l_01, swing1LimitDegrees=0, swing2LimitDegrees=20, twistLimitDegrees=0}
 {bone1Name=dyn_skirt_side_l_03, bone2Name=dyn_skirt_side_l_02, swing1LimitDegrees=0, swing2LimitDegrees=20, twistLimitDegrees=0}
 {bone1Name=dyn_skirt_side_l_04, bone2Name=dyn_skirt_side_l_03, swing1LimitDegrees=0, swing2LimitDegrees=30, twistLimitDegrees=0}
 {
  bone1Name=pelvis
  bone2Name=root
  swing1LimitDegrees=45
  swing1Motion=Free
  swing2LimitDegrees=45
  swing2Motion=Free
  twistLimitDegrees=45
  twistMotion=Free
 }
 {bone1Name=calf_r, bone2Name=thigh_r, swing1LimitDegrees=45, swing2LimitDegrees=45, twistLimitDegrees=45}
 {
  bone1Name=dyn_skirt_bk_l_04
  bone2Name=dyn_skirt_side_l_04
  swing1LimitDegrees=45
  swing1Motion=Free
  swing2LimitDegrees=45
  swing2Motion=Free
  twistLimitDegrees=45
  twistMotion=Free
 }
 {
  bone1Name=dyn_skirt_mid_bk_04
  bone2Name=dyn_skirt_bk_l_04
  swing1LimitDegrees=45
  swing1Motion=Free
  swing2LimitDegrees=45
  swing2Motion=Free
  twistLimitDegrees=45
  twistMotion=Free
 }
 {
  bone1Name=dyn_skirt_bk_r_04
  bone2Name=dyn_skirt_frnt_r_04
  swing1LimitDegrees=45
  swing1Motion=Free
  swing2LimitDegrees=45
  swing2Motion=Free
  twistLimitDegrees=45
  twistMotion=Free
 }
 {
  bone1Name=dyn_skirt_mid_bk_04
  bone2Name=dyn_skirt_bk_r_04
  swing1LimitDegrees=45
  swing1Motion=Free
  swing2LimitDegrees=45
  swing2Motion=Free
  twistLimitDegrees=45
  twistMotion=Free
 }
 {
  bone1Name=dyn_skirt_bk_l_03
  bone2Name=dyn_skirt_side_l_03
  swing1LimitDegrees=45
  swing1Motion=Free
  swing2LimitDegrees=45
  swing2Motion=Free
  twistLimitDegrees=45
  twistMotion=Free
 }
 {
  bone1Name=dyn_skirt_mid_bk_03
  bone2Name=dyn_skirt_bk_l_03
  swing1LimitDegrees=45
  swing1Motion=Free
  swing2LimitDegrees=45
  swing2Motion=Free
  twistLimitDegrees=45
  twistMotion=Free
 }
 {
  bone1Name=dyn_skirt_bk_r_03
  bone2Name=dyn_skirt_frnt_r_03
  swing1LimitDegrees=45
  swing1Motion=Free
  swing2LimitDegrees=45
  swing2Motion=Free
  twistLimitDegrees=45
  twistMotion=Free
 }
 {
  bone1Name=dyn_skirt_mid_bk_03
  bone2Name=dyn_skirt_bk_r_03
  swing1LimitDegrees=45
  swing1Motion=Free
  swing2LimitDegrees=45
  swing2Motion=Free
  twistLimitDegrees=45
  twistMotion=Free
 }
 {
  bone1Name=dyn_skirt_bk_r_02
  bone2Name=dyn_skirt_frnt_r_2
  swing1LimitDegrees=45
  swing1Motion=Free
  swing2LimitDegrees=45
  swing2Motion=Free
  twistLimitDegrees=45
  twistMotion=Free
 }
 {
  bone1Name=dyn_skirt_mid_bk_02
  bone2Name=dyn_skirt_bk_r_02
  swing1LimitDegrees=45
  swing1Motion=Free
  swing2LimitDegrees=45
  swing2Motion=Free
  twistLimitDegrees=45
  twistMotion=Free
 }
 {
  bone1Name=dyn_skirt_bk_l_02
  bone2Name=dyn_skirt_side_l_02
  swing1LimitDegrees=45
  swing1Motion=Free
  swing2LimitDegrees=45
  swing2Motion=Free
  twistLimitDegrees=45
  twistMotion=Free
 }
 {
  bone1Name=dyn_skirt_mid_bk_02
  bone2Name=dyn_skirt_bk_l_02
  swing1LimitDegrees=45
  swing1Motion=Free
  swing2LimitDegrees=45
  swing2Motion=Free
  twistLimitDegrees=45
  twistMotion=Free
 }
 {bone1Name=thigh_r, bone2Name=pelvis, swing1LimitDegrees=45, swing2LimitDegrees=45, twistLimitDegrees=45}
]
