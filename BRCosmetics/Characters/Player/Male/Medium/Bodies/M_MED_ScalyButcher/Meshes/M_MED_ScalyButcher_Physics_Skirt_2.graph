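# Physics asset for the skirt of the male medium ScalyButcher character body, with capsules on pelvis, legs and feet plus a large root sphere.
# Five skirt chains (back left, back middle, back right, front right, side left) hang from the pelvis via angularly free base joints, with mostly tight swing and twist limits below.
# Angularly free cross-constraints link neighboring chains at their tips, plus back-right to front-right mid-chain, tying the strips together.
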PhysicsAsset /BRCosmetics/Characters/Player/Male/Medium/Bodies/M_MED_ScalyButcher/Meshes/M_MED_ScalyButcher_Physics_Skirt_2
bodies {
 calf_l [{shapeType=Capsule, length=32.5656, radius=11.8556, center=<-20.6668 0 0.8731>, rotation=<rot 90 0 0>}]
 calf_r [{shapeType=Capsule, length=32.5396, radius=11.8816, center=<20.6668 0 -0.8731>, rotation=<rot 90 0 180>}]
 dyn_skirt_bk_l_01 [{shapeType=Capsule, length=12.5053, radius=10.4947, shapeName=dyn_skirt_bk_l_01_capsule, center=<3.1339 0 9.5526>, rotation=<rot 0 0 90>}]
 dyn_skirt_bk_l_02 [{shapeType=Capsule, length=15, radius=8, shapeName=dyn_skirt_bk_l_02_capsule, center=<6.5877 0 6.5581>, rotation=<rot 0 0 90>}]
 dyn_skirt_bk_l_03 [{shapeType=Capsule, length=15, radius=8, shapeName=dyn_skirt_bk_l_03_capsule, center=<6.5877 0 6.5581>, rotation=<rot 0 0 90>}]
 dyn_skirt_bk_l_04 [{shapeType=Capsule, length=15, radius=8, shapeName=dyn_skirt_bk_l_04_capsule, center=<6.5877 0 6.5581>, rotation=<rot 0 0 90>}]
 dyn_skirt_bk_l_base [{shapeType=Capsule, length=0.505, radius=0.505, shapeName=dyn_skirt_bk_l_base_capsule}]
 dyn_skirt_bk_r_01 [{shapeType=Capsule, length=12.8559, radius=10.1441, shapeName=dyn_skirt_bk_r_01_capsule, center=<-7.1194 0 -8.6734>, rotation=<rot 0 0 90>}]
 dyn_skirt_bk_r_02 [{shapeType=Capsule, length=15, radius=8, shapeName=dyn_skirt_bk_r_02_capsule, center=<-7.7817 0 -5.8965>, rotation=<rot 0 0 90>}]
 dyn_skirt_bk_r_03 [{shapeType=Capsule, length=15, radius=8, shapeName=dyn_skirt_bk_r_03_capsule, center=<-7.7817 0 -5.8965>, rotation=<rot 0 0 90>}]
 dyn_skirt_bk_r_04 [{
   shapeType=Capsule
   length=22.8841
   radius=7.6672
   shapeName=dyn_skirt_bk_r_04_capsule
   center=<-7.7817 0 -5.8965>
   rotation=<rot 0 -7.0578 90>
  }]
 dyn_skirt_bk_r_base [{shapeType=Capsule, length=0.505, radius=0.505, shapeName=dyn_skirt_bk_r_base_capsule}]
 dyn_skirt_frnt_r_01 [{shapeType=Capsule, length=12.3592, radius=10.6408, shapeName=dyn_skirt_frnt_r_01_capsule, center=<-8.3199 0 -10.0007>, rotation=<rot 0 0 90>}]
 dyn_skirt_frnt_r_03 [{shapeType=Capsule, length=15, radius=8, shapeName=dyn_skirt_frnt_r_03_capsule, center=<-8.3199 0 -4.7248>, rotation=<rot 0 0 90>}]
 dyn_skirt_frnt_r_04 [{
   shapeType=Capsule
   length=22.9031
   radius=6.8796
   shapeName=dyn_skirt_frnt_r_04_capsule
   center=<-9.8753 0 -6.4596>
   rotation=<rot 0.0001 1.5301 83.4375>
  }]
 dyn_skirt_frnt_r_2 [{shapeType=Capsule, length=15, radius=8, shapeName=dyn_skirt_frnt_r_2_capsule, center=<-8.3199 0 -6.0221>, rotation=<rot 0 0 90>}]
 dyn_skirt_frnt_r_base [{shapeType=Capsule, length=0.505, radius=0.505, shapeName=dyn_skirt_frnt_r_base_capsule}]
 dyn_skirt_mid_bk_01 [{shapeType=Capsule, length=12.4475, radius=10.5525, shapeName=dyn_skirt_mid_bk_01_capsule, center=<5.8468 -10.0686 0>}]
 dyn_skirt_mid_bk_02 [{shapeType=Capsule, length=15, radius=8, shapeName=dyn_skirt_mid_bk_02_capsule, center=<7.9261 -5.1314 0>}]
 dyn_skirt_mid_bk_03 [{shapeType=Capsule, length=23.158, radius=8, shapeName=dyn_skirt_mid_bk_03_capsule, center=<7.9261 -5.1314 0>}]
 dyn_skirt_mid_bk_04 [{
   shapeType=Capsule
   length=25.9597
   radius=8
   shapeName=dyn_skirt_mid_bk_04_capsule
   center=<7.9261 -5.1314 0>
   rotation=<rot 10.9971 0.0001 0>
  }]
 dyn_skirt_mid_bk_base [{shapeType=Capsule, length=0.505, radius=0.505, shapeName=dyn_skirt_mid_bk_base_capsule}]
 dyn_skirt_side_l_01 [{shapeType=Capsule, length=12.9633, radius=10.0367, shapeName=dyn_skirt_side_l_01_capsule, center=<2.4195 0 9.434>, rotation=<rot 0 0 90>}]
 dyn_skirt_side_l_02 [
  {shapeType=Capsule, length=15, radius=8, shapeName=dyn_skirt_side_l_02_capsule, center=<6.7328 0 7.702>, rotation=<rot 0 0 90>}
  {shapeType=Capsule, length=12, radius=4, shapeName=dyn_skirt_side_l_02_capsule, center=<6.9776 9.5775 1.0602>, rotation=<rot 90 0 0>}
 ]
 dyn_skirt_side_l_03 [{shapeType=Capsule, length=15, radius=8, shapeName=dyn_skirt_side_l_03_capsule, center=<6.7328 0 7.702>, rotation=<rot 0 0 90>}]
 dyn_skirt_side_l_04 [{shapeType=Capsule, length=15, radius=8, shapeName=dyn_skirt_side_l_04_capsule, center=<6.7328 0 7.702>, rotation=<rot 0 0 90>}]
 dyn_skirt_side_l_base [{shapeType=Capsule, length=0.505, radius=0.505, shapeName=dyn_skirt_side_l_base_capsule}]
 foot_l [{shapeType=Capsule, length=15, radius=10, center=<-4.3187 10.289 0>, rotation=<rot 0 0 90>}]
 foot_r [{shapeType=Capsule, length=15, radius=10, center=<4.3187 -10.289 0>, rotation=<rot 0 0 -90>}]
 pelvis [{shapeType=Capsule, length=8.7003, radius=12.0776, center=<-5.7681 0 0>}]
 root [{shapeType=Sphere, radius=150, shapeName=root_sphere, center=<0 -10.7566 -138.6542>}]
 thigh_l [{shapeType=Capsule, length=30.9043, radius=10.8742, center=<-19.9205 0 1.074>, rotation=<rot 88.6963 179.9986 179.9988>}]
 thigh_r [{shapeType=Capsule, length=30.9043, radius=10.8742, center=<19.9205 0 -1.074>, rotation=<rot 88.6963 179.9986 -0.0012>}]
}
constraints [
 {bone1Name=thigh_l, bone2Name=pelvis, swing1LimitDegrees=45, swing2LimitDegrees=45, twistLimitDegrees=45}
 {bone1Name=calf_l, bone2Name=thigh_l, swing1LimitDegrees=45, swing2LimitDegrees=45, twistLimitDegrees=45}
 {bone1Name=foot_l, bone2Name=calf_l, swing1LimitDegrees=45, swing2LimitDegrees=45, twistLimitDegrees=45}
 {bone1Name=foot_r, bone2Name=calf_r, swing1LimitDegrees=45, swing2LimitDegrees=45, twistLimitDegrees=45}
 {bone1Name=calf_r, bone2Name=thigh_r, swing1LimitDegrees=45, swing2LimitDegrees=45, twistLimitDegrees=45}
 {bone1Name=thigh_r, bone2Name=pelvis, swing1LimitDegrees=45, swing2LimitDegrees=45, twistLimitDegrees=45}
 {
  bone1Name=dyn_skirt_bk_l_base
  bone2Name=pelvis
  swing1LimitDegrees=45
  swing1Motion=Free
  swing2LimitDegrees=45
  swing2Motion=Free
  twistLimitDegrees=45
  twistMotion=Free
 }
 {bone1Name=dyn_skirt_bk_l_01, bone2Name=dyn_skirt_bk_l_base, swing1LimitDegrees=20, swing2LimitDegrees=20, twistLimitDegrees=5}
 {bone1Name=dyn_skirt_bk_l_02, bone2Name=dyn_skirt_bk_l_01, swing1LimitDegrees=20, swing2LimitDegrees=20, twistLimitDegrees=5}
 {bone1Name=dyn_skirt_bk_l_03, bone2Name=dyn_skirt_bk_l_02, swing1LimitDegrees=20, swing2LimitDegrees=20, twistLimitDegrees=5}
 {bone1Name=dyn_skirt_bk_l_04, bone2Name=dyn_skirt_bk_l_03, swing1LimitDegrees=20, swing2LimitDegrees=20, twistLimitDegrees=5}
 {
  bone1Name=dyn_skirt_bk_r_base
  bone2Name=pelvis
  swing1LimitDegrees=45
  swing1Motion=Free
  swing2LimitDegrees=45
  swing2Motion=Free
  twistLimitDegrees=45
  twistMotion=Free
 }
 {bone1Name=dyn_skirt_bk_r_01, bone2Name=dyn_skirt_bk_r_base, swing1LimitDegrees=20, swing2LimitDegrees=20, twistLimitDegrees=5}
 {bone1Name=dyn_skirt_bk_r_02, bone2Name=dyn_skirt_bk_r_01, swing1LimitDegrees=20, swing2LimitDegrees=20, twistLimitDegrees=5}
 {bone1Name=dyn_skirt_bk_r_03, bone2Name=dyn_skirt_bk_r_02, swing1LimitDegrees=20, swing2LimitDegrees=20, twistLimitDegrees=5}
 {bone1Name=dyn_skirt_bk_r_04, bone2Name=dyn_skirt_bk_r_03, swing1LimitDegrees=20, swing2LimitDegrees=20, twistLimitDegrees=5}
 {
  bone1Name=dyn_skirt_frnt_r_base
  bone2Name=pelvis
  swing1LimitDegrees=45
  swing1Motion=Free
  swing2LimitDegrees=45
  swing2Motion=Free
  twistLimitDegrees=45
  twistMotion=Free
 }
 {bone1Name=dyn_skirt_frnt_r_01, bone2Name=dyn_skirt_frnt_r_base, swing1LimitDegrees=20, swing2LimitDegrees=20, twistLimitDegrees=10}
 {bone1Name=dyn_skirt_frnt_r_2, bone2Name=dyn_skirt_frnt_r_01, swing1LimitDegrees=20, swing2LimitDegrees=20, twistLimitDegrees=10}
 {bone1Name=dyn_skirt_frnt_r_03, bone2Name=dyn_skirt_frnt_r_2, swing1LimitDegrees=20, swing2LimitDegrees=20, twistLimitDegrees=10}
 {bone1Name=dyn_skirt_frnt_r_04, bone2Name=dyn_skirt_frnt_r_03, swing1LimitDegrees=20, swing2LimitDegrees=30, twistLimitDegrees=1}
 {
  bone1Name=dyn_skirt_mid_bk_base
  bone2Name=pelvis
  swing1LimitDegrees=45
  swing1Motion=Free
  swing2LimitDegrees=45
  swing2Motion=Free
  twistLimitDegrees=45
  twistMotion=Free
 }
 {bone1Name=dyn_skirt_mid_bk_01, bone2Name=dyn_skirt_mid_bk_base, swing1LimitDegrees=20, swing2LimitDegrees=20, twistLimitDegrees=5}
 {bone1Name=dyn_skirt_mid_bk_02, bone2Name=dyn_skirt_mid_bk_01, swing1LimitDegrees=20, swing2LimitDegrees=20, twistLimitDegrees=5}
 {bone1Name=dyn_skirt_mid_bk_03, bone2Name=dyn_skirt_mid_bk_02, swing1LimitDegrees=20, swing2LimitDegrees=20, twistLimitDegrees=5}
 {bone1Name=dyn_skirt_mid_bk_04, bone2Name=dyn_skirt_mid_bk_03, swing1LimitDegrees=20, swing2LimitDegrees=20, twistLimitDegrees=5}
 {
  bone1Name=dyn_skirt_side_l_base
  bone2Name=pelvis
  swing1LimitDegrees=45
  swing1Motion=Free
  swing2LimitDegrees=45
  swing2Motion=Free
  twistLimitDegrees=45
  twistMotion=Free
 }
 {
  bone1Name=dyn_skirt_side_l_01
  bone2Name=dyn_skirt_side_l_base
  swing1LimitDegrees=45
  swing1Motion=Free
  swing2LimitDegrees=45
  swing2Motion=Free
  twistLimitDegrees=45
  twistMotion=Free
 }
 {bone1Name=dyn_skirt_side_l_02, bone2Name=dyn_skirt_side_l_01, swing1LimitDegrees=10, swing2LimitDegrees=20, twistLimitDegrees=5}
 {bone1Name=dyn_skirt_side_l_03, bone2Name=dyn_skirt_side_l_02, swing1LimitDegrees=10, swing2LimitDegrees=20, twistLimitDegrees=5}
 {bone1Name=dyn_skirt_side_l_04, bone2Name=dyn_skirt_side_l_03, swing1LimitDegrees=10, swing2LimitDegrees=20, twistLimitDegrees=5}
 {
  bone1Name=pelvis
  bone2Name=root
  swing1LimitDegrees=45
  swing1Motion=Free
  swing2LimitDegrees=45
  swing2Motion=Free
  twistLimitDegrees=45
  twistMotion=Free
 }
 {
  bone1Name=dyn_skirt_bk_l_04
  bone2Name=dyn_skirt_side_l_04
  swing1LimitDegrees=45
  swing1Motion=Free
  swing2LimitDegrees=45
  swing2Motion=Free
  twistLimitDegrees=45
  twistMotion=Free
 }
 {
  bone1Name=dyn_skirt_mid_bk_04
  bone2Name=dyn_skirt_bk_r_04
  swing1LimitDegrees=45
  swing1Motion=Free
  swing2LimitDegrees=45
  swing2Motion=Free
  twistLimitDegrees=45
  twistMotion=Free
 }
 {
  bone1Name=dyn_skirt_mid_bk_04
  bone2Name=dyn_skirt_bk_l_04
  swing1LimitDegrees=45
  swing1Motion=Free
  swing2LimitDegrees=45
  swing2Motion=Free
  twistLimitDegrees=45
  twistMotion=Free
 }
 {
  bone1Name=dyn_skirt_bk_r_04
  bone2Name=dyn_skirt_frnt_r_04
  swing1LimitDegrees=45
  swing1Motion=Free
  swing2LimitDegrees=45
  swing2Motion=Free
  twistLimitDegrees=45
  twistMotion=Free
 }
 {
  bone1Name=dyn_skirt_bk_r_02
  bone2Name=dyn_skirt_frnt_r_2
  swing1LimitDegrees=45
  swing1Motion=Free
  swing2LimitDegrees=45
  swing2Motion=Free
  twistLimitDegrees=45
  twistMotion=Free
 }
]
