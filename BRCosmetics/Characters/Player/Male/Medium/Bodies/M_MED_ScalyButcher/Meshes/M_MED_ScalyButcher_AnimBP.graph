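# Secondary-motion AnimBP for the ScalyButcher male body cosmetic, animating its skirt, ribbons and sword.
# The parent mesh pose feeds a movement- and state-driven Control Rig, then two skirt RigidBody sims blended by crouch/DBNO, then further skirt and ribbon sims.
# Sword AnimDynamics, ribbon constraints, conditional cape and physics-off offsets, and fixed additive skirt tweaks finish the pose.
# The EventGraph updates these states, easing frontend alphas.
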AnimBlueprint /BRCosmetics/Characters/Player/Male/Medium/Bodies/M_MED_ScalyButcher/Meshes/M_MED_ScalyButcher_AnimBP
target_skeleton=&/M_MED_ScalyButcher_Skeleton
graphs [
 AnimGraph {nodes [
   n1 = "Transform (Modify) Bone - Bone: dyn_skirt_frnt_r_01"(ComponentPose=Local To Component(LocalPose=Use cached pose 'toAlphaFT'()), Alpha=1, Translation="0,0,0", Rotation="3,0,0", Scale="1,1,1") {
    rotationMode=BMM_Additive
    translationMode=BMM_Additive
    alphaBoolBlend {blendOption=Linear}
    boneToModify {boneName=dyn_skirt_frnt_r_01}
    rotation {pitch=3}
   }
   n2 = "Transform (Modify) Bone - Bone: dyn_skirt_frnt_r_2"(ComponentPose=@n1, Alpha=1, Translation="0,0,0", Rotation="0,0,0", Scale="1,1,1") {rotationMode=BMM_Additive, translationMode=BMM_Additive, alphaBoolBlend {blendOption=Linear}, boneToModify {boneName=dyn_skirt_frnt_r_2}}
   n3 = "Transform (Modify) Bone - Bone: dyn_skirt_side_l_04"(ComponentPose=@n2, Alpha=1, Translation="0,0,0", Rotation="-10.408,0.000,-0.000", Scale="1,1,1") {
    rotationMode=BMM_Additive
    translationMode=BMM_Additive
    alphaBoolBlend {blendOption=Linear}
    boneToModify {boneName=dyn_skirt_side_l_04}
    rotation {pitch=-10.408}
   }
   Output Pose(Result=Component To Local(ComponentPose=@n3)) {layerGroup=None}
   n4 = RigidBody!(ComponentPose=Local To Component(LocalPose=Use cached pose 'ToPhysics'()), Alpha, OverrideWorldGravity:Vector="0,0,0") -> (Pose) {
    bClampLinearTranslationLimitToRefPose=true
    bOverrideWorldGravity=true
    overridePhysicsAsset=&/M_MED_ScalyButcher_Physics_Skirt_2
    simulationSpace=BaseBoneSpace
    alphaBoolBlend {blendOption=Linear}
    baseBoneRef {boneName=pelvis}
    simSpaceSettings {
     maxAngularAcceleration=400
     maxAngularVelocity=3
     maxLinearAcceleration=500
     maxLinearVelocity=250
     velocityScaleZ=0
     worldAlpha=1
     externalLinearDragV=<0.1 0.1>
     externalLinearVelocity=<0.1 0.1>
    }
   }
   n5 = Component To Local(ComponentPose=@n4)
   n6 = RigidBody(
    ComponentPose=Local To Component(LocalPose=Use cached pose 'ToPhysics'())
    Alpha=Get Is Not Frontend() -> (Is Not Frontend:double)
    OverrideWorldGravity="0,0,-980"
   ) {
    bClampLinearTranslationLimitToRefPose=true
    bOverrideWorldGravity=true
    overridePhysicsAsset=&/M_MED_ScalyButcher_Physics_Skirt
    simulationSpace=BaseBoneSpace
    alphaBoolBlend {blendOption=Linear}
    baseBoneRef {boneName=pelvis}
    componentAppliedLinearAccClamp=<3000 3000 3000>
    componentLinearAccScale=<0.8 0.4 0.4>
    componentLinearVelScale=<0.8 0.4 0.4>
    overrideWorldGravity {z=-980}
    simSpaceSettings {
     maxAngularAcceleration=400
     maxAngularVelocity=3
     maxLinearAcceleration=500
     maxLinearVelocity=250
     velocityScaleZ=0
     worldAlpha=1
     externalLinearDragV=<0.1 0.1>
     externalLinearVelocity=<0.1 0.1>
    }
   }
   n7 = Component To Local(ComponentPose=@n6)
   n8 = Get Sit still/DBNO() -> (Sit still/DBNO:bool)
   n9 = RigidBody(
    ComponentPose=Local To Component(LocalPose=Blend Poses by bool(BlendPose_0=@n5, BlendPose_1=@n7, bActiveValue=@n8) {blendType=Linear})
    OverrideWorldGravity="0,0,-980"
   ) {
    bClampLinearTranslationLimitToRefPose=true
    bOverrideWorldGravity=true
    overridePhysicsAsset=&/M_MED_ScalyButcher_Physics_Skirt_Frontend
    simulationSpace=BaseBoneSpace
    alphaBoolBlend {blendOption=Linear}
    baseBoneRef {boneName=pelvis}
    componentAppliedLinearAccClamp=<1500 1500 1500>
    componentLinearAccScale=<0.3 0.2 0.2>
    componentLinearVelScale=<0.3 0.2 0.2>
    overrideWorldGravity {z=-980}
    simSpaceSettings {maxAngularAcceleration=200, maxAngularVelocity=1, maxLinearAcceleration=300, maxLinearVelocity=150, velocityScaleZ=0, worldAlpha=1}
   }
   n10 = RigidBody(ComponentPose=@n9, OverrideWorldGravity="0,0,0") {
    bClampLinearTranslationLimitToRefPose=true
    bOverrideWorldGravity=true
    overridePhysicsAsset=&/M_MED_ScalyButcher_Physics_Ribbons
    simulationSpace=BaseBoneSpace
    alphaBoolBlend {blendOption=Linear}
    baseBoneRef {boneName=pelvis}
    componentLinearAccScale=<1.4 1.3 1.3>
    componentLinearVelScale=<1.4 1.3 1.3>
    simSpaceSettings {
     maxAngularAcceleration=400
     maxAngularVelocity=3
     maxLinearAcceleration=500
     maxLinearVelocity=250
     velocityScaleZ=0
     worldAlpha=1
     externalLinearDragV=<0.1 0.1>
     externalLinearVelocity=<0.1 0.1>
    }
   }
   Save cached pose 'toAnimD'(Pose=Component To Local(ComponentPose=@n10)) {cacheName=toAnimD}
   n11 = Control Rig!(
    Source=Copy Pose From Mesh() {bUseAttachedParent=true}
    Alpha
    Forward Backward Movement:double=Get Forward Backward Movement() -> (Forward Backward Movement:double)
    Left Right Movement:double=Get Left Right Movement() -> (Left Right Movement:double)
    DBNO:bool=Get bIsDBNO()
    Knee Sliding:bool=Get bIsKneeSliding()
    Baseball Sliding:bool=Get bIsBaseballSliding()
    Parachuting:bool=Get bIsParachuteOpened()
    Playing emote:bool=Get bIsPlayingEmote()
    RBAN off:bool=Get RBAN_off() -> (RBAN_off:bool)
    Is Alpha Frontend:double=Get Is Alpha Frontend() -> (Is Alpha Frontend:double)
   ) -> (Pose) {
    alphaBoolBlend {blendOption=Linear}
    controlRigAssetReference {blueprintRigClass=&/M_MED_ScalyButcher_CtrlRig.M_MED_ScalyButcher_CtrlRig_C}
   }
   Save cached pose 'ToPhysics'(Pose=@n11) {cacheName=ToPhysics}
   n12 = "Transform (Modify) Bone - Bone: dyn_sword_base"!(
    ComponentPose=Local To Component(LocalPose=Use cached pose 'toAnimD'())
    bAlphaBoolEnabled:bool=Get isWearingCape() -> (isWearingCape:bool)
    Translation="-0.823,-0.036,-1.313"
    Rotation="4.127,20.334,-11.087"
    Scale="1,1,1"
   ) -> (Pose) {
    alphaInputType=Bool
    rotationMode=BMM_Additive
    rotationSpace=BCS_BoneSpace
    translationMode=BMM_Additive
    translationSpace=BCS_BoneSpace
    alphaBoolBlend {blendOption=Linear}
    boneToModify {boneName=dyn_sword_base}
    rotation=<rot 4.127 20.334 -11.0871>
    translation=<-0.823 -0.036 -1.313>
   }
   n13 = Anim Dynamics - Bone: dyn_sword_01!(ComponentPose=@n12, Alpha=Get Is Not Frontend(), GravityOverride:Vector="0,0,0", ExternalForce="0,0,0") -> (Pose) {
    angularDampingOverride=0.9999
    bOverrideAngularDamping=true
    bOverrideLinearDamping=true
    bUseGravityOverride=true
    linearDampingOverride=0.999
    simulationSpace=BoneRelative
    alphaBoolBlend {blendOption=Linear}
    boundBone {boneName=dyn_sword_01}
    physicsBodyDefinitions [{
      boundBone {boneName=dyn_sword_01}
      boxExtents=<2 2 2>
      constraintSetup {angularLimitsMax=<10 10 5>, angularLimitsMin=<-10 -10 -5>, angularTarget=<0.1 0.1 0>}
      localJointOffset=<0.1 0 0>
     }]
    relativeSpaceBone {boneName=spine_01}
   }
   n14 = "Constraint - dyn_ribbon_frnt_02 (PP)"!(ComponentPose=@n13, Alpha=0.5, ConstraintWeights_0=0.3, ConstraintWeights_1:float=0.7) -> (Pose) {
    alphaBoolBlend {blendOption=Linear}
    boneToModify {boneName=dyn_ribbon_frnt_02}
    constraintSetup [{targetBone {boneName=dyn_skirt_bk_r_02}}, {targetBone {boneName=dyn_skirt_frnt_r_2}}]
    constraintWeights [1,1]
   }
   n15 = "Constraint - dyn_ribbon_bk_02 (P)"(ComponentPose=@n14, Alpha=0.5, ConstraintWeights_0=1) {
    alphaBoolBlend {blendOption=Linear}
    boneToModify {boneName=dyn_ribbon_bk_02}
    constraintSetup [{targetBone {boneName=dyn_skirt_bk_r_02}}]
    constraintWeights [1]
   }
   n16 = "Transform (Modify) Bone - Bone: skirt_support_bk_ALT"!(ComponentPose=@n15, bAlphaBoolEnabled:bool=Get RBAN_off(), Translation="0,-9.326,9", Rotation="0,0,0", Scale="1,1,1") -> (Pose) {
    alphaInputType=Bool
    rotationMode=BMM_Additive
    translationMode=BMM_Additive
    alphaBoolBlend {blendOption=Linear}
    boneToModify {boneName=skirt_support_bk_ALT}
    translation {y=-9.326, z=9}
   }
   Save cached pose 'toMove'(Pose=Component To Local(ComponentPose=@n16)) {cacheName=toMove}
   n17 = "Transform (Modify) Bone - Bone: dyn_skirt_mid_bk_03"(
    ComponentPose=Local To Component(LocalPose=Use cached pose 'toMove'())
    Alpha=1
    Translation="0,-3,0"
    Rotation="-0.000,0.000,13.588"
    Scale="1,1,1"
   ) {
    rotationMode=BMM_Additive
    translationMode=BMM_Additive
    alphaBoolBlend {blendOption=Linear}
    alphaScaleBiasClamp {bInterpResult=true}
    boneToModify {boneName=dyn_skirt_mid_bk_03}
    rotation {roll=13.588}
    translation {y=-3}
   }
   n18 = "Transform (Modify) Bone - Bone: dyn_skirt_bk_r_03"(ComponentPose=@n17, Alpha=1, Translation="0,-3,0", Rotation="10,0,12.662", Scale="1,1,1") {
    rotationMode=BMM_Additive
    translationMode=BMM_Additive
    alphaBoolBlend {blendOption=Linear}
    alphaScaleBiasClamp {bInterpResult=true}
    boneToModify {boneName=dyn_skirt_bk_r_03}
    rotation {pitch=10, roll=12.662}
    translation {y=-3}
   }
   n19 = "Transform (Modify) Bone - Bone: dyn_skirt_bk_l_03"(ComponentPose=@n18, Alpha=1, Translation="0,-6,0", Rotation="-0.000,0.000,11.955", Scale="1,1,1") {
    rotationMode=BMM_Additive
    translationMode=BMM_Additive
    alphaBoolBlend {blendOption=Linear}
    alphaScaleBiasClamp {bInterpResult=true}
    boneToModify {boneName=dyn_skirt_bk_l_03}
    rotation {roll=11.955}
    translation {y=-6}
   }
   n20 = "Transform (Modify) Bone - Bone: dyn_skirt_frnt_r_04"(ComponentPose=@n19, Alpha=1, Translation="0,0,0", Rotation="-10,0,0", Scale="1,1,1") {
    rotationMode=BMM_Additive
    translationMode=BMM_Additive
    alphaBoolBlend {blendOption=Linear}
    alphaScaleBiasClamp {bInterpResult=true}
    boneToModify {boneName=dyn_skirt_frnt_r_04}
    rotation {pitch=-10}
   }
   n21 = "Transform (Modify) Bone - Bone: dyn_skirt_frnt_r_03"(ComponentPose=@n20, Alpha=1, Translation="-2,0,0", Rotation="0,0,0", Scale="1,1,1") {
    rotationMode=BMM_Additive
    translationMode=BMM_Additive
    alphaBoolBlend {blendOption=Linear}
    alphaScaleBiasClamp {bInterpResult=true}
    boneToModify {boneName=dyn_skirt_frnt_r_03}
    translation {x=-2}
   }
   n22 = "Transform (Modify) Bone - Bone: dyn_skirt_bk_r_04"(ComponentPose=@n21, Alpha=1, Translation="0,0,0", Rotation="20,0,12", Scale="1,1,1") {
    rotationMode=BMM_Additive
    translationMode=BMM_Additive
    alphaBoolBlend {blendOption=Linear}
    alphaScaleBiasClamp {bInterpResult=true}
    boneToModify {boneName=dyn_skirt_bk_r_04}
    rotation {pitch=20, roll=12}
   }
   n23 = "Transform (Modify) Bone - Bone: dyn_skirt_side_l_04"(ComponentPose=@n22, Alpha=1, Translation="2,0,0", Rotation="0,0,0", Scale="1,1,1") {
    rotationMode=BMM_Additive
    translationMode=BMM_Additive
    alphaBoolBlend {blendOption=Linear}
    alphaScaleBiasClamp {bInterpResult=true}
    boneToModify {boneName=dyn_skirt_side_l_04}
    translation {x=2}
   }
   n24 = "Transform (Modify) Bone - Bone: dyn_skirt_bk_l_04"(ComponentPose=@n23, Alpha=1, Translation="3,0,0", Rotation="0,0,0", Scale="1,1,1") {
    rotationMode=BMM_Additive
    translationMode=BMM_Additive
    alphaBoolBlend {blendOption=Linear}
    alphaScaleBiasClamp {bInterpResult=true}
    boneToModify {boneName=dyn_skirt_bk_l_04}
    translation {x=3}
   }
   Save cached pose 'toAlphaFT'(Pose=Component To Local(ComponentPose=@n24)) {cacheName=toAlphaFT}
  ]}
 EventGraph {nodes [
   n1 = Event BlueprintUpdateAnimation()
   > Set Delta Time X(Delta Time X:double=@n1.DeltaTimeX) -> (Output_Get:double)
   > n2 = Sequence!(execute) -> (then_0, then_1, then_2:Exec)
   n3 = Get Crouching() -> (Crouching:bool)
   Set Sit still/DBNO(execute=@n2.then_1, Sit still/DBNO:bool=OR Boolean!(A=Get bIsDBNO(), B=@n3, C:bool) -> (ReturnValue)) -> (Output_Get:bool)
   > Set isWearingCape(isWearingCape=CheckForBackbling().bIsBackblingACape)
   n4 = AND Boolean(A=Get bIsInFrontEnd(), B=NOT Boolean(A=Get bIsPlayingEmote()))
   > n5 = Branch(Condition=NOT Boolean(A=@n4))
   > Set Is Not Frontend(Is Not Frontend:double=FInterpTo(Current=Get Is Not Frontend(), Target=1, DeltaTime=Get Delta Time X() -> (Delta Time X:double), InterpSpeed=5)) -> (Output_Get:double)
   Set Is Not Frontend(execute=@n5.else, Is Not Frontend=FInterpTo(Current=Get Is Not Frontend(), Target=0, DeltaTime=Get Delta Time X(), InterpSpeed=5))
   Set RBAN_off(execute=@n2.then_2, RBAN_off:bool=NOT Boolean(A=Get bRigidBodyEnabled())) -> (Output_Get:bool)
   n6 = OR Boolean!(
    A="To Bool (Double)"!(Value:double=Get Left Right Movement()) -> (ReturnValue:bool)
    B="To Bool (Double)"(Value=Get Forward Backward Movement())
   ) -> (ReturnValue)
   n7 = Get Delta Time X()
   > Set Movement 30fps(Movement 30fps:double=FInterpTo(
     Current=Get Movement 30fps() -> (Movement 30fps:double)
     Target="To Float (Boolean)"(InBool=@n6)
     DeltaTime=float * float(A=@n7, B=0.3) {is=title K2Node_PromotableOperator}
     InterpSpeed=10
    )) -> (Output_Get:double)
   > n8 = Branch(Condition=@n4)
   > Set Is Alpha Frontend(Is Alpha Frontend:double=FInterpTo(Current=Get Is Alpha Frontend(), Target=1, DeltaTime=Get Delta Time X(), InterpSpeed=5)) -> (Output_Get:double)
   Set Is Alpha Frontend(execute=@n8.else, Is Alpha Frontend=FInterpTo(Current=Get Is Alpha Frontend(), Target=0, DeltaTime=Get Delta Time X(), InterpSpeed=5))
   n9 = Gravity Override State Function DEPRECATED(
    execute=@n2.then_0
    CharacterPartAnimInstance=Self-Reference()
    GameTimeInSeconds=GetGameTimeInSeconds()
    GravityOverrideParams=Get Gravity Override Params_0() -> (Gravity Override Params_0:Gravity Override Params Struct Structure)
   )
   > Set Gravity Override(Gravity Override=@n9)
   > n10 = PlayerMovement(AnimInstance=Self-Reference(), XDotProductCoefficient=1, YDotProductCoefficient=1, ZDotProductCoefficient=1)
   > Set Left Right Movement(Left Right Movement:double=@n10) -> (Output_Get:double)
   > Set Forward Backward Movement(Forward Backward Movement:double=@n10.YMovement) -> (Output_Get:double)
   Event OnEndCrouching()
   > Set Crouching(Crouching:bool=false) -> (Output_Get:bool)
   Event OnBeginCrouching()
   > Set Crouching(Crouching=true)
   n11 = "To Bool (Double)"(Value=Get Left Right Movement())
   n12 = "To Bool (Double)"(Value=Get Forward Backward Movement())
   AND Boolean(A=@n3, B=NOT Boolean(A=OR Boolean(A=@n11, B=@n12)))
   Get bIsInFrontEnd()
  ]}
]
variables [
 Crouching {type=boolean}
 Delta Time X {type=number}
 Forward Backward Movement {type=number}
 Gravity Override {type=Vector}
 Gravity Override Params_0 {
  type=GravityOverrideParamsStruct
  default {
   skydive_DivingDown {windFrequency=<1.5 4 2>}
   skydive_FloatingDown {gravityOverride=<0 500 -500>, windAmplitude=<50 -200 200>, windFrequency=<1.5 4 8>, windOffset=<0 -300 50>}
   skydive_GoingUp {gravityOverride=<0 500 -300>, windAmplitude=<50 -300 150>, windFrequency=<1.5 3 6>, windOffset=<0 0 100>}
   skydive_Parachuting {gravityOverride=<0 -500 -2000>, windAmplitude=<200 -50 100>}
  }
 }
 Is Alpha Frontend {type=number}
 Is Not Frontend {type=number}
 isWearingCape {type=boolean}
 Left Right Movement {type=number}
 Movement 30fps {type=number}
 RBAN_off {type=boolean}
 Sit still/DBNO {type=boolean}
]
