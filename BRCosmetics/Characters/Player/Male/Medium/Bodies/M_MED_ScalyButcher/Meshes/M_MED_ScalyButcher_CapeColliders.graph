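# Physics asset for the medium male ScalyButcher body, giving capsule colliders to the body, rear skirt chains and sword, named as cape colliders.
# Capsules cover the spine, pelvis, arms, hands, legs and feet.
# Three rear skirt chains (left, middle, right) and a two-bone sword chain hang from the pelvis with free swing and twist.
# Body joints leave motion at default with 45-degree limit values, except the free right calf-to-thigh joint.
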PhysicsAsset /BRCosmetics/Characters/Player/Male/Medium/Bodies/M_MED_ScalyButcher/Meshes/M_MED_ScalyButcher_CapeColliders
bodies {
 calf_l [{shapeType=Capsule, length=27.9444, radius=13.9313, center=<-16.4529 -2.2427 0>, rotation=<rot 90 0 0>}]
 calf_r [{shapeType=Capsule, length=27.9444, radius=13.9313, center=<16.4529 2.2427 0>, rotation=<rot 90 180 0>}]
 dyn_skirt_bk_l_01 [{shapeType=Capsule, length=15.8313, radius=9.4535, shapeName=dyn_skirt_bk_l_01_capsule, center=<8.7045 0 0>, rotation=<rot 90 0 0>}]
 dyn_skirt_bk_l_02 [{shapeType=Capsule, length=15.8313, radius=9.4535, shapeName=dyn_skirt_bk_l_02_capsule, center=<8.7045 0 0>, rotation=<rot 90 0 0>}]
 dyn_skirt_bk_l_03 [{shapeType=Capsule, length=15.8313, radius=9.4535, shapeName=dyn_skirt_bk_l_03_capsule, center=<8.7045 0 0>, rotation=<rot 90 0 0>}]
 dyn_skirt_bk_l_04 [{shapeType=Capsule, length=13.365, radius=11.5776, shapeName=dyn_skirt_bk_l_04_capsule, center=<8.7045 0 0>, rotation=<rot 90 0 0>}]
 dyn_skirt_bk_l_base [{shapeType=Capsule, length=0, radius=0, shapeName=dyn_skirt_bk_l_base_capsule, rotation=<rot 90 0 0>}]
 dyn_skirt_bk_r_01 [{shapeType=Capsule, length=15.8313, radius=9.4535, shapeName=dyn_skirt_bk_r_01_capsule, center=<-10.2758 0 0>, rotation=<rot 90 0 0>}]
 dyn_skirt_bk_r_02 [{shapeType=Capsule, length=15.8313, radius=9.4535, shapeName=dyn_skirt_bk_r_02_capsule, center=<-10.2758 0 0>, rotation=<rot 90 0 0>}]
 dyn_skirt_bk_r_03 [{shapeType=Capsule, length=15.8313, radius=9.4535, shapeName=dyn_skirt_bk_r_03_capsule, center=<-10.2758 0 0>, rotation=<rot 90 0 0>}]
 dyn_skirt_bk_r_04 [{shapeType=Capsule, length=13.365, radius=11.5776, shapeName=dyn_skirt_bk_r_04_capsule, center=<-10.2758 0 0>, rotation=<rot 90 0 0>}]
 dyn_skirt_bk_r_base [{shapeType=Capsule, length=0, radius=0, shapeName=dyn_skirt_bk_r_base_capsule, rotation=<rot 90 0 0>}]
 dyn_skirt_mid_bk_01 [{shapeType=Capsule, length=15.8313, radius=9.4535, shapeName=dyn_skirt_mid_bk_01_capsule, center=<10.613 0 0>, rotation=<rot 90 0 0>}]
 dyn_skirt_mid_bk_02 [{shapeType=Capsule, length=15.8313, radius=9.4535, shapeName=dyn_skirt_mid_bk_02_capsule, center=<10.613 0 0>, rotation=<rot 90 0 0>}]
 dyn_skirt_mid_bk_03 [{shapeType=Capsule, length=15.8313, radius=9.4535, shapeName=dyn_skirt_mid_bk_03_capsule, center=<10.613 0 0>, rotation=<rot 90 0 0>}]
 dyn_skirt_mid_bk_04 [{shapeType=Capsule, length=13.365, radius=11.5776, shapeName=dyn_skirt_mid_bk_04_capsule, center=<10.613 0 0>, rotation=<rot 90 0 0>}]
 dyn_skirt_mid_bk_base [{shapeType=Capsule, length=0, radius=0, shapeName=dyn_skirt_mid_bk_base_capsule, rotation=<rot 90 0 0>}]
 dyn_sword_01 [{
   shapeType=Capsule
   length=57.5909
   radius=10.3126
   shapeName=dyn_sword_01_capsule
   center=<0.3772 7.1031 4.6985>
   rotation=<rot 0 164.1225 -55.1669>
  }]
 dyn_sword_base [{shapeType=Capsule, length=0.505, radius=0.505, shapeName=dyn_sword_base_capsule, rotation=<rot 0 -90 0>}]
 foot_l [{shapeType=Capsule, length=13.1207, radius=9.0416, center=<-4.1955 6.0815 0>, rotation=<rot 0 14.9998 90>}]
 foot_r [{shapeType=Capsule, length=13.1207, radius=9.0416, center=<4.1955 -6.0815 0>, rotation=<rot 0 14.9998 -90>}]
 hand_l [{shapeType=Capsule, length=9.5015, radius=6.7933, center=<6.2703 0 0>, rotation=<rot 90 0 0>}]
 hand_r [{shapeType=Capsule, length=9.5015, radius=6.7933, center=<-6.2703 0 0>, rotation=<rot 90 0 180>}]
 lowerarm_l [{shapeType=Capsule, length=14.3387, radius=7.5574, center=<8.9852 0 0.2267>, rotation=<rot 90 0 0>}]
 lowerarm_r [{shapeType=Capsule, length=14.3387, radius=7.5574, center=<-8.9852 0 -0.2267>, rotation=<rot 90 0 180>}]
 pelvis [{shapeType=Capsule, length=9.0004, radius=17.7863, center=<-2.4855 -0.9262 0>}]
 spine_02 [{shapeType=Capsule, length=1.9888, radius=16.1591, center=<-2.6142 0 0>}]
 spine_03 [{shapeType=Capsule, length=6.4463, radius=15.9965}]
 spine_04 [{shapeType=Capsule, length=4.8231, radius=15.5172}]
 spine_05 [{shapeType=Capsule, length=14.0449, radius=15.8769}]
 thigh_l [{shapeType=Capsule, length=22.1612, radius=13.3386, center=<-18.4217 -1.038 0>, rotation=<rot 90 180 2.664>}]
 thigh_r [{shapeType=Capsule, length=22.1612, radius=13.3386, center=<18.4217 1.038 0>, rotation=<rot 90 180 -177.336>}]
 upperarm_l [{shapeType=Capsule, length=17.1712, radius=7.7501, center=<10.0455 0 0>, rotation=<rot 90 0 0>}]
 upperarm_r [{shapeType=Capsule, length=17.1712, radius=7.7501, center=<-10.0455 0 0>, rotation=<rot 90 0 180>}]
}
constraints [
 {bone1Name=spine_02, bone2Name=pelvis, swing1LimitDegrees=45, swing2LimitDegrees=45, twistLimitDegrees=45}
 {bone1Name=spine_03, bone2Name=spine_02, swing1LimitDegrees=45, swing2LimitDegrees=45, twistLimitDegrees=45}
 {bone1Name=spine_04, bone2Name=spine_03, swing1LimitDegrees=45, swing2LimitDegrees=45, twistLimitDegrees=45}
 {bone1Name=spine_05, bone2Name=spine_04, swing1LimitDegrees=45, swing2LimitDegrees=45, twistLimitDegrees=45}
 {bone1Name=upperarm_l, bone2Name=spine_05, swing1LimitDegrees=45, swing2LimitDegrees=45, twistLimitDegrees=45}
 {bone1Name=lowerarm_l, bone2Name=upperarm_l, swing1LimitDegrees=45, swing2LimitDegrees=45, twistLimitDegrees=45}
 {bone1Name=hand_l, bone2Name=lowerarm_l, swing1LimitDegrees=45, swing2LimitDegrees=45, twistLimitDegrees=45}
 {bone1Name=foot_l, bone2Name=calf_l, swing1LimitDegrees=45, swing2LimitDegrees=45, twistLimitDegrees=45}
 {bone1Name=thigh_l, bone2Name=pelvis, swing1LimitDegrees=45, swing2LimitDegrees=45, twistLimitDegrees=45}
 {bone1Name=hand_r, bone2Name=lowerarm_r, swing1LimitDegrees=45, swing2LimitDegrees=45, twistLimitDegrees=45}
 {bone1Name=foot_r, bone2Name=calf_r, swing1LimitDegrees=45, swing2LimitDegrees=45, twistLimitDegrees=45}
 {bone1Name=upperarm_r, bone2Name=spine_05, swing1LimitDegrees=45, swing2LimitDegrees=45, twistLimitDegrees=45}
 {bone1Name=lowerarm_r, bone2Name=upperarm_r, swing1LimitDegrees=45, swing2LimitDegrees=45, twistLimitDegrees=45}
 {bone1Name=calf_l, bone2Name=thigh_l, swing1LimitDegrees=45, swing2LimitDegrees=45, twistLimitDegrees=45}
 {
  bone1Name=dyn_sword_01
  bone2Name=dyn_sword_base
  swing1LimitDegrees=45
  swing1Motion=Free
  swing2LimitDegrees=45
  swing2Motion=Free
  twistLimitDegrees=45
  twistMotion=Free
 }
 {
  bone1Name=dyn_sword_base
  bone2Name=pelvis
  swing1LimitDegrees=45
  swing1Motion=Free
  swing2LimitDegrees=45
  swing2Motion=Free
  twistLimitDegrees=45
  twistMotion=Free
 }
 {bone1Name=thigh_r, bone2Name=pelvis, swing1LimitDegrees=45, swing2LimitDegrees=45, twistLimitDegrees=45}
 {
  bone1Name=calf_r
  bone2Name=thigh_r
  swing1LimitDegrees=45
  swing1Motion=Free
  swing2LimitDegrees=45
  swing2Motion=Free
  twistLimitDegrees=45
  twistMotion=Free
 }
 {
  bone1Name=dyn_skirt_bk_l_base
  bone2Name=pelvis
  swing1LimitDegrees=45
  swing1Motion=Free
  swing2LimitDegrees=45
  swing2Motion=Free
  twistLimitDegrees=45
  twistMotion=Free
 }
 {
  bone1Name=dyn_skirt_bk_l_01
  bone2Name=dyn_skirt_bk_l_base
  swing1LimitDegrees=45
  swing1Motion=Free
  swing2LimitDegrees=45
  swing2Motion=Free
  twistLimitDegrees=45
  twistMotion=Free
 }
 {
  bone1Name=dyn_skirt_bk_l_02
  bone2Name=dyn_skirt_bk_l_01
  swing1LimitDegrees=45
  swing1Motion=Free
  swing2LimitDegrees=45
  swing2Motion=Free
  twistLimitDegrees=45
  twistMotion=Free
 }
 {
  bone1Name=dyn_skirt_bk_l_03
  bone2Name=dyn_skirt_bk_l_02
  swing1LimitDegrees=45
  swing1Motion=Free
  swing2LimitDegrees=45
  swing2Motion=Free
  twistLimitDegrees=45
  twistMotion=Free
 }
 {
  bone1Name=dyn_skirt_bk_l_04
  bone2Name=dyn_skirt_bk_l_03
  swing1LimitDegrees=45
  swing1Motion=Free
  swing2LimitDegrees=45
  swing2Motion=Free
  twistLimitDegrees=45
  twistMotion=Free
 }
 {
  bone1Name=dyn_skirt_bk_r_base
  bone2Name=pelvis
  swing1LimitDegrees=45
  swing1Motion=Free
  swing2LimitDegrees=45
  swing2Motion=Free
  twistLimitDegrees=45
  twistMotion=Free
 }
 {
  bone1Name=dyn_skirt_bk_r_01
  bone2Name=dyn_skirt_bk_r_base
  swing1LimitDegrees=45
  swing1Motion=Free
  swing2LimitDegrees=45
  swing2Motion=Free
  twistLimitDegrees=45
  twistMotion=Free
 }
 {
  bone1Name=dyn_skirt_bk_r_02
  bone2Name=dyn_skirt_bk_r_01
  swing1LimitDegrees=45
  swing1Motion=Free
  swing2LimitDegrees=45
  swing2Motion=Free
  twistLimitDegrees=45
  twistMotion=Free
 }
 {
  bone1Name=dyn_skirt_bk_r_03
  bone2Name=dyn_skirt_bk_r_02
  swing1LimitDegrees=45
  swing1Motion=Free
  swing2LimitDegrees=45
  swing2Motion=Free
  twistLimitDegrees=45
  twistMotion=Free
 }
 {
  bone1Name=dyn_skirt_bk_r_04
  bone2Name=dyn_skirt_bk_r_03
  swing1LimitDegrees=45
  swing1Motion=Free
  swing2LimitDegrees=45
  swing2Motion=Free
  twistLimitDegrees=45
  twistMotion=Free
 }
 {
  bone1Name=dyn_skirt_mid_bk_base
  bone2Name=pelvis
  swing1LimitDegrees=45
  swing1Motion=Free
  swing2LimitDegrees=45
  swing2Motion=Free
  twistLimitDegrees=45
  twistMotion=Free
 }
 {
  bone1Name=dyn_skirt_mid_bk_04
  bone2Name=dyn_skirt_mid_bk_03
  swing1LimitDegrees=45
  swing1Motion=Free
  swing2LimitDegrees=45
  swing2Motion=Free
  twistLimitDegrees=45
  twistMotion=Free
 }
 {
  bone1Name=dyn_skirt_mid_bk_03
  bone2Name=dyn_skirt_mid_bk_02
  swing1LimitDegrees=45
  swing1Motion=Free
  swing2LimitDegrees=45
  swing2Motion=Free
  twistLimitDegrees=45
  twistMotion=Free
 }
 {
  bone1Name=dyn_skirt_mid_bk_02
  bone2Name=dyn_skirt_mid_bk_01
  swing1LimitDegrees=45
  swing1Motion=Free
  swing2LimitDegrees=45
  swing2Motion=Free
  twistLimitDegrees=45
  twistMotion=Free
 }
 {
  bone1Name=dyn_skirt_mid_bk_01
  bone2Name=dyn_skirt_mid_bk_base
  swing1LimitDegrees=45
  swing1Motion=Free
  swing2LimitDegrees=45
  swing2Motion=Free
  twistLimitDegrees=45
  twistMotion=Free
 }
]
