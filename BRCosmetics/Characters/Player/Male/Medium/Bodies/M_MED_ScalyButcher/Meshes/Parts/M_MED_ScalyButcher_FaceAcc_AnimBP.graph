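# Animation Blueprint for the ScalyButcher body's face accessory mesh, adding secondary motion to the back of its scarf.
# It copies the attached parent's pose, simulates dyn_scarf_bk_01 with AnimDynamics constrained by a spherical limit on spine_04, then additively offsets dyn_scarf_bk_base.
# Each update sets a Gravity Override variable through a deprecated state function, but the AnimGraph uses its own fixed gravity and never reads that variable.
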AnimBlueprint /BRCosmetics/Characters/Player/Male/Medium/Bodies/M_MED_ScalyButcher/Meshes/Parts/M_MED_ScalyButcher_FaceAcc_AnimBP
target_skeleton=&/M_MED_ScalyButcher_FaceAcc_Skeleton
graphs [
 AnimGraph {nodes [
   n1 = Anim Dynamics - Bone: dyn_scarf_bk_01!(
    ComponentPose=Local To Component(LocalPose=Copy Pose From Mesh() {bUseAttachedParent=true})
    Alpha=1
    GravityOverride:Vector="0,0,-980"
    ExternalForce="0,0,0"
   ) -> (Pose) {
    angularDampingOverride=1
    bOverrideAngularDamping=true
    bUseGravityOverride=true
    bUsePlanarLimit=false
    bUseSphericalLimits=true
    alphaBoolBlend {blendOption=Linear}
    boundBone {boneName=dyn_scarf_bk_01}
    componentLinearAccScale=<0.3 0.3>
    componentLinearVelScale=<0.3 0.3>
    gravityOverride {z=-980}
    physicsBodyDefinitions [{
      boundBone {boneName=dyn_scarf_bk_01}
      boxExtents=<2 2 2>
      constraintSetup {angularLimitsMax=<0 10 10>, angularLimitsMin=<0 0 0>, angularTarget=<0 0 0>}
      localJointOffset=<0.1 0 0>
     }]
    relativeSpaceBone {boneName=spine_05}
    sphericalLimits [{limitRadius=23.4657, drivingBone {boneName=spine_04}, sphereLocalOffset=<0 8.1501 0>}]
   }
   n2 = "Transform (Modify) Bone - Bone: dyn_scarf_bk_base"(ComponentPose=@n1, Alpha=1, Translation="0,-2,0", Rotation="0,0,0", Scale="1,1,1") {
    rotationMode=BMM_Additive
    translationMode=BMM_Additive
    alphaBoolBlend {blendOption=Linear}
    boneToModify {boneName=dyn_scarf_bk_base}
    translation {y=-2}
   }
   Output Pose(Result=Component To Local(ComponentPose=@n2)) {layerGroup=None}
  ]}
 EventGraph {nodes [
   Event BlueprintUpdateAnimation()
   > n1 = Gravity Override State Function DEPRECATED(
    CharacterPartAnimInstance=Self-Reference()
    GameTimeInSeconds=GetGameTimeInSeconds()
    GravityOverrideParams=Get Gravity Override Params_0() -> (Gravity Override Params_0:Gravity Override Params Struct Structure)
   )
   > Set Gravity Override(Gravity Override=@n1)
  ]}
]
variables [
 Gravity Override {type=Vector}
 Gravity Override Params_0 {type=GravityOverrideParamsStruct}
]
